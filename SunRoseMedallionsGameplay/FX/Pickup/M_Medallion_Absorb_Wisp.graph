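# Unlit translucent Niagara mesh-particle material for the soul wisp effect of medallion absorption on pickup.
# Dynamic parameters swipe and widen a soul texture, which blends with time-scrolled, per-particle-offset Voronoi smoke inside sphere masks.
# The result indexes a color curve atlas, with hue shift and particle color, while soul greyscale also drives an offset along vertex normals.
# Opacity combines erosion, particle alpha and inverted fresnel; mobile skips smoke, and lower quality levels skip depth fade.
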Material /SunRoseMedallionsGameplay/FX/Pickup/M_Medallion_Absorb_Wisp
graph {
 properties {
  bUsedWithNiagaraMeshParticles=true
  blendMode=BLEND_Translucent
  pixelDepthOffsetMode=PDOM_Legacy
  shadingModel=MSM_Unlit
  twoSided=true
  usedShadingModels=Unlit
 }
 nodes [
  n1 = ScalarParameter() {parameterName=HueShiftPercentage}
  n2 = MaterialFunctionCall(Float2=TextureCoordinate()) {materialFunction=&BreakOutFloat2Components}
  n3 = DynamicParameter() {defaultValue=<rgba 0 1 0 1>, paramNames [SoulSwipe,SoulWidth,Erode,Param4]}
  n4 = AppendVector(A=Add(A=Multiply(A=@n2.R) {constB=1.5}, B=@n3.SoulSwipe) {constB=0}, B=Multiply(A=@n2.G, B=@n3.SoulWidth))
  n5 = MaterialFunctionCall(Float2=@n4) {materialFunction=&BreakOutFloat2Components}
  n6 = Multiply(A=@n5.R) {constB=0.3}
  n7 = SphereMask(A=AppendVector(A=Add(A=@n6) {constB=0.22}, B=@n5.G), B=Constant() {r=0.5}) {attenuationRadius=0.45, hardnessPercent=25}
  n8 = MaterialFunctionCall(Float2=TextureCoordinate()) {materialFunction=&BreakOutFloat2Components}
  n9 = Multiply(A=Power(Base=@n8.R)) {constB=0.8}
  n10 = Time()
  n11 = ParticleRandom()
  n12 = Add(A=@n8.R, B=Multiply(A=Add(A=@n10, B=@n11)) {constB=0.35})
  n13 = TextureSampleParameter2D(UVs=AppendVector(A=Multiply(A=@n12) {constB=1.7}, B=Multiply(A=@n8.G) {constB=2})) {parameterName=SmokeTex, texture=/Game/Effects/Fort_Effects/Textures/T_PinkSpike_Voronoi2D_Height}
  n14 = Multiply(
   A=SphereMask(A=AppendVector(A=@n9, B=@n8.G), B=Constant2Vector() {g=0.5, r=0.3}) {attenuationRadius=0.4, hardnessPercent=0}
   B=Power(Base=OneMinus(None=@n13.G))
  ) {constA=1}
  n15 = DynamicParameter() {defaultValue=<rgba 0 1 0 1>, paramNames [SoulSwipe,SoulWidth,Erode,Param4]}
  n16 = Power(Base=OneMinus(None=Add(A=@n8.R, B=@n15.SoulSwipe) {constB=0}))
  n17 = LinearInterpolate(
   A=ShadingPathSwitch(Default=@n14, Deferred=@n14, Mobile=Constant() {r=0.35}) {desc=Migrated from FeatureLevelSwitch}
   B=Power(Base=NamedRerouteUsage())
   Alpha=OneMinus(None=Saturate(None=@n16))
  )
  n18 = CurveAtlasRowParameter(CurveTime=Saturate(None=Multiply(A=@n17, B=@n7))) {
   atlas=/SunRoseMedallionsGameplay/FX/Pickup/Curves/CurveColorAtlas_GodMedallions
   curve=/SunRoseMedallionsGameplay/FX/Pickup/Curves/Curve_Color_AbsorbSouls_Green
   defaultValue=2
   parameterName=ColorCurve
  }
  MaterialFunctionCall(Emissive=Multiply(
    A=MaterialFunctionCall(Hue Shift Percentage=@n1, Texture=@n18) {materialFunction=&HueShift}.Result
    B=ParticleColor().RGB
   )) {materialFunction=&MF_FX_EmissiveAdjustments}
  n19 = Power(Base=ConstantBiasScale(None=NamedRerouteUsage())) {constExponent=1.6}
  Multiply(A=Multiply(A=Saturate(None=@n19), B=VertexNormalWS())) {constB=8}
  n20 = Multiply(
   A=Multiply(A=Power(Base=@n17) {constExponent=0.4}, B=ParticleColor().A)
   B=Subtract(A=@n7, B=DynamicParameter() {defaultValue=<rgba 0 1 0 1>, paramNames [SoulSwipe,SoulWidth,Erode,Param4]}.Erode)
  )
  n21 = DepthFade(Opacity=@n20) {fadeDistanceDefault=5}
  n22 = QualitySwitch(Default=@n21, Low=@n20, High=@n21, Medium=@n20, Epic=@n21)
  n23 = Fresnel() {baseReflectFraction=0, exponent=0.5}
  Multiply(
   A=ShadingPathSwitch(Default=@n22, Deferred=@n22, Mobile=@n20) {desc=Migrated from FeatureLevelSwitch}
   B=Power(Base=OneMinus(None=@n23)) {constExponent=1}
  )
  n24 = TextureSampleParameter2D(UVs=Clamp(None=@n4)) {parameterName=SoulTexture, samplerType=SAMPLERTYPE_LinearGrayscale, texture=&/T_AbsorbMedallion_Soul}
  NamedRerouteDeclaration(None=@n24.G) {name=SoulTextureGreyscale, nodeColor=<rgba 0 1 0.7844 1>}
 ]
}
