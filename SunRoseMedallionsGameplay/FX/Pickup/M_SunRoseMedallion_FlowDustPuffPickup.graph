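# Unlit translucent Niagara particle material for the Sun Rose Medallion pickup's flowing dust puff, in warm orange-gold with hot sparkles.
# Noise-distorted smoke drives the color ramp, while panned sand dots, boosted by Perlin noise, add HDR glints with dynamic-parameter speed.
# An 8x8 mist flipbook over particle life, with particle color, shapes brightness and depth-faded opacity.
# Quality and shading-path switches replace noise with constants on cheaper tiers and drop depth fade on mobile.
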Material /SunRoseMedallionsGameplay/FX/Pickup/M_SunRoseMedallion_FlowDustPuffPickup
graph {
 properties {
  bUsedWithNiagaraMeshParticles=true
  bUsedWithNiagaraSprites=true
  blendMode=BLEND_Translucent
  pixelDepthOffsetMode=PDOM_Legacy
  shadingModel=MSM_Unlit
  twoSided=true
  usedShadingModels=Unlit
 }
 nodes [
  n1 = NamedRerouteUsage()
  n2 = Power(Base=@n1) {constExponent=3}
  n3 = LinearInterpolate(
   A=LinearInterpolate(A=Constant3Vector() {constant=<rgba 0.4 0.1973 0.08 1>}, B=Constant3Vector() {constant=<rgba 1 0.6239 0.1941 1>}, Alpha=Saturate(None=@n1))
   B=Constant3Vector() {constant=<rgba 2 0.8533 0.28 1>}
   Alpha=Saturate(None=Multiply(A=@n2) {constB=2})
  )
  n4 = Constant3Vector() {constant=<rgba 30 18.75 7.5 1>}
  n5 = Add(A=TextureCoordinate(), B=ParticleRandom())
  n6 = DynamicParameter() {defaultValue=<rgba 0 1 1 1>, paramNames [SpeedSparkles, StartingFrameUV, Param3, Param4]}
  n7 = Constant() {r=-0.1}
  n8 = TextureSampleParameter2D(UVs=Panner(Coordinate=Multiply(A=@n5) {constB=0.9}, Speed=AppendVector(A=@n6.SpeedSparkles, B=@n7)) {speedX=0.1, speedY=-0.1}) {parameterName=TexSand, texture=&/T_SandDust_Dots_Pickup_Small}
  n9 = TextureSampleParameter2D(UVs=Panner(Coordinate=Multiply(A=@n5) {constB=4}) {speedX=0.1, speedY=0.2}) {
   parameterName=TexNoise2
   samplerType=SAMPLERTYPE_Masks
   texture=/Game/Characters/Player/Male/Large/Bodies/M_LRG_Clash_V/Textures/Simple_Perlin_M
  }
  n10 = LinearInterpolate(Alpha=Saturate(None=Power(Base=@n9.R))) {constA=1, constB=16}
  n11 = Constant() {r=3}
  n12 = QualitySwitch(Default=@n10, Low=@n11, High=@n10, Medium=@n11, Epic=@n10)
  n13 = Multiply(
   A=LinearInterpolate(A=@n8.R, B=@n8.R) {constAlpha=0.4}
   B=ShadingPathSwitch(Default=@n12, Deferred=@n12, Mobile=@n11) {desc=Migrated from FeatureLevelSwitch}
  )
  n14 = NamedRerouteUsage()
  n15 = Saturate(None=Power(Base=Multiply(A=@n13, B=@n14)) {constExponent=3})
  n16 = TextureObject() {samplerType=SAMPLERTYPE_LinearColor, texture=/Game/Effects/Fort_Effects/Textures/Smoke/T_FallingMist_8x8_01}
  n17 = TextureCoordinate()
  n18 = Constant2Vector() {g=8, r=8}
  n19 = Add(
   A=DynamicParameter() {defaultValue=<rgba 0 1 1 1>, paramNames [SpeedSparkles, StartingFrameUV, Param3, Param4]}.StartingFrameUV
   B=Multiply(A=ParticleRelativeTime()) {constB=32}
  )
  n20 = ComponentMask(None=Power(Base=MaterialFunctionCall!(Texture=@n16, UVs=@n17, SubImages=@n18, Frame=@n19) {materialFunction=/Engine/Functions/Engine_MaterialFunctions03/Texturing/SubUV_Function}.RGB) {constExponent=1.5}) {r=true}
  n21 = LinearInterpolate(Alpha=@n20) {constA=0.3}
  n22 = ParticleColor()
  MaterialFunctionCall(Emissive=Multiply(A=LinearInterpolate(A=@n3, B=@n4, Alpha=@n15), B=Multiply(A=@n21, B=@n22.RGB))) {materialFunction=&MF_FX_EmissiveAdjustments}
  n23 = Constant() {r=0.75}
  n24 = ParticleRandom()
  n25 = Add(A=TextureCoordinate(), B=@n24)
  n26 = TextureSampleParameter2D(UVs=Panner(Coordinate=@n25) {speedX=0.1, speedY=0.05}) {parameterName=TexNoise, samplerType=SAMPLERTYPE_LinearColor, texture=/Game/Effects/Fort_Effects/Textures/Tile/T_PerlinRGB_Displacement}
  n27 = Multiply(A=@n26.R) {constB=0.2}
  n28 = Add(A=Add(A=@n24) {constB=0.33}, B=TextureCoordinate())
  n29 = Power(Base=TextureSampleParameter2D(UVs=Add(A=@n27, B=@n28)) {parameterName=TexSmoke, texture=/Game/Effects/Fort_Effects/Textures/Tile/T_Inky_Smoke_Tile2}.R) {constExponent=0.3}
  n30 = QualitySwitch(Default=@n29, Low=@n23, High=@n29, Medium=@n23, Epic=@n29)
  NamedRerouteDeclaration(None=ShadingPathSwitch(Default=@n30, Deferred=@n30, Mobile=@n23) {desc=Migrated from FeatureLevelSwitch}) {name=Smoke Greyscale, nodeColor=<rgba 0.3594 0 1 1>}
  Panner(Coordinate=@n25) {speedX=0.13, speedY=0.07}
  n31 = Multiply(A=@n22.A, B=@n20)
  n32 = DepthFade(Opacity=@n31) {fadeDistanceDefault=15}
  ShadingPathSwitch(Default=@n32, Deferred=@n32, Mobile=@n31) {desc=Migrated from FeatureLevelSwitch}
 ]
}
